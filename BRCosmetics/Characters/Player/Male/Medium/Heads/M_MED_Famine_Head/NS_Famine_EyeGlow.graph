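# Niagara eye effect for the Famine character head: glowing green eyes wreathed in black smoke.
# Flare bursts two persistent mirrored particles with sprite and light renderers, scaled by user glow and light strength parameters.
# Flipbook smoke sprites thin out as owner speed passes a threshold, while a second trail emitter starts spawning and inherits scaled owner velocity.
# All emitters set VisibilityTag from bisHeadHidden.
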NiagaraSystem /BRCosmetics/Characters/Player/Male/Medium/Heads/M_MED_Famine_Head/NS_Famine_EyeGlow
dependencies {usedDataInterfaces [&NiagaraDataInterfaceCurve], usedDynamicInputs [], usedModules [&SystemState]}
emitters {
 Black_Smoke001 {
  emitterUpdateScript [
   &EmitterState(Life Cycle Mode=::NewEnumerator1, Loop Duration=1, UseLoopDelay=false)
   &SpawnRate(
    SpawnRate=&MakeCustomFloatFromBool(
     Bool=&SetBoolByFloatComparison(A=&VectorLength(Vector=$Engine.Owner.Velocity), B=10)
     True Float=4
     False Float=7
    )
    Use Spawn Probability=false
   )
  ]
  particleSpawnScript [
   /Niagara/Modules/Spawn/Initialization/InitializeParticle(
    Lifetime=&UniformRangedFloat(Minimum=0.8, Maximum=1.8)
    Mesh Scale=&Multiply_Vector(
     A=&UniformRangedVector(Minimum=<0.3 0.125 1>, Maximum=<0.5 0.25 1>)
     B=&VectorFromFloat(Value=1.25)
    )
    Module.Write Color=true
    Module.Write Lifetime=true
    Module.Write Mass=true
    Module.Write Material Random=true
    Module.Write Position=false
    Module.Write Scale=true
    Module.Write SpriteRotation=true
    Module.Write SpriteSize=true
    Position=&SimulationPosition()
    Sprite Rotation=&RandomRangeFloat(Minimum=0, Maximum=360)
    Sprite Size=&Vector2DFromFloat(Value=25)
   )
   &AddVelocity(Velocity=&UniformRangedVector(Minimum=<-4 0 -11>, Maximum=<25 1 11>))
   &CylinderLocation(
    Cylinder Height=5
    Cylinder Radius=2
    MaskSpawnGroup=false
    Module.Enable Lathe Profile=false
    Module.Override Randomness=false
    Module.Surface Only=false
    Offset=<5 0 0>
    Orientation Axis:ENiagaraOrientationAxis=::XAxis
   )
  ]
  particleUpdateScript [
   &ParticleState()
   &Drag(Use Linear Drag=true, Use Rotational Drag=false)
   &SolveForcesAndVelocity(Clamp Velocity=false, Limit Acceleration=false)
   &Color(
    Color=<rgba 0 0 0 1>
    Scale Alpha=&FloatFromCurve(
     FloatCurve=DataInterface() {propertyValues {Curve {keys [
         {interpMode=RCIM_Cubic, leaveTangent=0.1765, tangentMode=RCTM_Break, time=0, value=0}
         {arriveTangent=0.6, interpMode=RCIM_Cubic, leaveTangent=0.1235, tangentMode=RCTM_Break, time=0.23, value=0.5}
         {interpMode=RCIM_Cubic, tangentMode=RCTM_Break, time=1, value=0}
        ]}}}
     Scale Curve=1.15
    )
   )
   &DynamicMaterialParameters(
    Index 0 Param 1=&FloatFromCurve(FloatCurve=DataInterface() {propertyValues {Curve {keys [<at 0 1>,<at 1 0>]}}})
    Index 0 Param 2=&UniformRangedFloat(Minimum=20, Maximum=24)
    Index 0 Param 4=&FloatFromCurve(FloatCurve=DataInterface() {propertyValues {Curve {keys [<at 0.0123 0.25>,<at 0.99 -0.3>,<at 1 -0.4>]}}})
    Module.Param0WriteEnabled=true
    Module.Param1WriteEnabled=true
    Module.Param2WriteEnabled=true
    Module.Param3WriteEnabled=true
   )
   /Niagara/Modules/Update/Orientation/V2/UpdateMeshOrientation(Rotation Rate=0.2, Rotation Vector:Vector3f=<1 0 0>)
   &ScaleMeshSize(Scale Factor=&VectorFromFloat(Value=&FloatFromCurve(FloatCurve=DataInterface() {propertyValues {Curve {keys [<at 0 0.5>,<at 1 1>]}}})))
   &SpriteRotationRate(Rotation Rate=&RandomRangeFloat(Minimum=-250, Maximum=-100))
   &SubUVAnimation(
    End Frame:NiagaraInt32=&RandomRangeInt(Minimum=45, Maximum=63)
    Start Frame:NiagaraInt32=&RandomRangeInt(Minimum=10, Maximum=20)
    SubUV Animation Mode:ENiagara_SubUVLookupMode=::NewEnumerator1
    SubUV Curve Index=HlslExpression() {hlslExpression=Particles.NormalizedAge / 1.5}
   )
   /SetVariables(Particles.VisibilityTag=&MakeCustomIntFromBool(Bool=$User.bisHeadHidden, True INT=1, False INT=0)) {isSetParameters=true, module=SetVariables_#1}
  ]
  rendererClasses [&NiagaraSpriteRendererProperties]
  renderers [{rendererClass=&NiagaraSpriteRendererProperties}]
 }
 Black_Smoke_WS_Trail001 {
  emitterUpdateScript [
   &EmitterState(Life Cycle Mode=::NewEnumerator1, Loop Duration=1, UseLoopDelay=false)
   &SpawnRate(
    SpawnRate=&MakeCustomFloatFromBool(
     Bool=&SetBoolByFloatComparison(A=&VectorLength(Vector=$Engine.Owner.Velocity), B=10)
     True Float=$User.BlackSmokeTrailSpawnRate
     False Float=0
    )
    Use Spawn Probability=false
   )
  ]
  particleSpawnScript [
   &InitializeParticle(
    Lifetime=&UniformRangedFloat(Minimum=1, Maximum=1.4)
    Mesh Scale=&Multiply_Vector(
     A=&UniformRangedVector(Minimum=<0.3 0.125 1>, Maximum=<0.5 0.25 1>)
     B=&VectorFromFloat(Value=1.25)
    )
    Module.Write Color=true
    Module.Write Lifetime=true
    Module.Write Mass=true
    Module.Write Material Random=true
    Module.Write Position=false
    Module.Write Scale=true
    Module.Write SpriteRotation=true
    Module.Write SpriteSize=true
    Position=&SimulationPosition()
    Sprite Rotation=&RandomRangeFloat(Minimum=0, Maximum=360)
    Sprite Size=&Vector2DFromFloat(Value=30)
   )
   &CylinderLocation(
    Cylinder Height=5
    Cylinder Radius=1
    MaskSpawnGroup=false
    Module.Enable Lathe Profile=false
    Module.Override Randomness=false
    Module.Surface Only=false
    Offset=<5 0 0>
    Orientation Axis:ENiagaraOrientationAxis=::XAxis
   )
   &AddVelocity(
    Scale Added Velocity=<0.05 -0.15 -0.1>
    Velocity=&TransformVector(Vector=$Engine.Owner.Velocity, Source Space=::World, Destination Space=::Local)
   )
  ]
  particleUpdateScript [
   &ParticleState()
   &SolveForcesAndVelocity(Clamp Velocity=false, Limit Acceleration=false)
   &Color(
    Color=<rgba 0 0 0 1>
    Scale Alpha=&FloatFromCurve(
     FloatCurve=DataInterface() {propertyValues {Curve {keys [
         {interpMode=RCIM_Cubic, leaveTangent=2.727, tangentMode=RCTM_Break, time=0, value=0}
         {arriveTangent=0.6, interpMode=RCIM_Cubic, leaveTangent=0.1235, tangentMode=RCTM_Break, time=0.23, value=0.9}
         {interpMode=RCIM_Cubic, tangentMode=RCTM_Break, time=1, value=0}
        ]}}}
     CurveIndex=$Particles.Age
     Scale Curve=0.8
    )
   )
   &DynamicMaterialParameters(
    Index 0 Param 1=&FloatFromCurve(FloatCurve=DataInterface() {propertyValues {Curve {keys [<at 0 1>,<at 1 0>]}}})
    Index 0 Param 2=&UniformRangedFloat(Minimum=20, Maximum=24)
    Index 0 Param 3=&FloatFromCurve(FloatCurve=DataInterface() {propertyValues {Curve {keys [<at 0 0>,<at 1 1>]}}})
    Index 0 Param 4=&FloatFromCurve(FloatCurve=DataInterface() {propertyValues {Curve {keys [<at 0.0123 0.25>,<at 0.99 -0.3>,<at 1 -1>]}}})
    Module.Param0WriteEnabled=true
    Module.Param1WriteEnabled=true
    Module.Param2WriteEnabled=true
    Module.Param3WriteEnabled=true
   )
   &SpriteRotationRate(Rotation Rate=&RandomRangeFloat(Minimum=-250, Maximum=-100))
   &SubUVAnimation(
    End Frame:NiagaraInt32=&RandomRangeInt(Minimum=45, Maximum=63)
    Start Frame:NiagaraInt32=&RandomRangeInt(Minimum=10, Maximum=20)
    SubUV Animation Mode:ENiagara_SubUVLookupMode=::NewEnumerator1
    SubUV Curve Index=HlslExpression() {hlslExpression=Particles.NormalizedAge / 1.5}
   )
   /SetVariables(Particles.VisibilityTag=&MakeCustomIntFromBool(Bool=$User.bisHeadHidden, True INT=1, False INT=0)) {isSetParameters=true, module=SetVariables_#1}
  ]
  rendererClasses [&NiagaraSpriteRendererProperties]
  renderers [{rendererClass=&NiagaraSpriteRendererProperties}]
 }
 Flare {
  emitterUpdateScript [
   &EmitterState(
    Inactive Response=::NewEnumerator1
    Life Cycle Mode=::NewEnumerator1
    Loop Behavior=::NewEnumerator1
    Loop Duration Mode=::NewEnumerator1
    UseLoopDelay=false
   )
   &SpawnBurst_Instantaneous(Spawn Count=1, Use Spawn Probability=false)
   &SpawnBurst_Instantaneous(Spawn Count=1, Spawn Group=1, Use Spawn Probability=false) {module=SpawnBurst_Instantaneous001}
  ]
  particleSpawnScript [
   &InitializeParticle(
    Mesh Scale=<1.3 1.3 1.3>
    Module.Write Color=true
    Module.Write Lifetime=true
    Module.Write Mass=true
    Module.Write Material Random=true
    Module.Write Position=false
    Module.Write Scale=false
    Module.Write SpriteRotation=false
    Module.Write SpriteSize=true
    Position=&SimulationPosition()
    Sprite Size=&Vector2DFromFloat(Value=35)
   )
   &SystemLocation(MaskSpawnGroup=false, Offset=<7.5538 7.9 4.5>)
   &SystemLocation(MaskSpawnGroup=true, Offset=<7.5538 7.9 -4.5>, Spawn Group Mask=1) {module=SystemLocation001}
   /SetVariables(Particles.Light_Position:Vector3f=&Add_Vector(A=Linked() {linkedVariable {name=Particles.Position, type=NiagaraPosition}}, B=<0.7844 0.5754 0>)) {isSetParameters=true}
  ]
  particleUpdateScript [
   &ParticleState(Kill Particles When Lifetime Has Elapsed=false, Let Infinitely Lived Particles Die When Emitter Deactivates=true)
   /SetVariables(Particles.EyeGlow_Color:LinearColor=&MultiplyLinearColor(
     Linear Color=<rgba 0 0.6146 0.0347 0.4>
     Scale Factor=&MakeFloat4FromVectorAndFloat("Vector (RGB)"=<1 1 1>, "Float (Alpha)"=$User.EyeGlowStrength)
    )) {isSetParameters=true}
   /SetVariables(Particles.LightGlow:LinearColor=&MultiplyLinearColor(
     Linear Color=<rgba 0 10 0.4508 1>
     Scale Factor=&MakeFloat4FromVectorAndFloat(
      "Vector (RGB)"=&VectorFromFloat(Value=&Multiply_Float(A=$User.LightStrength, B=0.2))
      "Float (Alpha)"=1
     )
    )) {isSetParameters=true}
   &CameraOffset(Camera Offset Amount=1) {enabled=false}
   &SolveForcesAndVelocity(Clamp Velocity=false, Limit Acceleration=false)
   /SetVariables(Particles.VisibilityTag=&MakeCustomIntFromBool(Bool=$User.bisHeadHidden, True INT=1, False INT=0)) {isSetParameters=true, module=SetVariables_#1}
  ]
  rendererClasses [&NiagaraSpriteRendererProperties, &NiagaraLightRendererProperties]
  renderers [
   {rendererClass=&NiagaraSpriteRendererProperties}
   {rendererClass=&NiagaraLightRendererProperties, rendererIndex=1}
  ]
 }
}
summary {
 systemName=NS_Famine_EyeGlow
 emitters [
  {emitterName=Flare, rendererClasses [&NiagaraSpriteRendererProperties, &NiagaraLightRendererProperties]}
  {emitterName=Black_Smoke001, rendererClasses [&NiagaraSpriteRendererProperties]}
  {emitterName=Black_Smoke_WS_Trail001, rendererClasses [&NiagaraSpriteRendererProperties]}
 ]
 userVariables [
  User.bisHeadHidden {type=NiagaraBool}
  User.BlackSmokeTrailSpawnRate {type=NiagaraFloat, defaultValue=10, description=""}
  User.EyeGlowMultiplier {type=NiagaraFloat, defaultValue=0, description=""}
  User.EyeGlowStrength {type=NiagaraFloat, defaultValue=0, description=""}
  User.LightStrength {type=NiagaraFloat, defaultValue=0, description=""}
  User.Right2ImpactVelocity {type=Vector3f, description="", defaultValue {struct=Vector3f, value=<0 0 0>}}
 ]
}
